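# Translucent unlit Niagara sprite material for the PrismParticle pickaxe impact, rendering a reflective splash from normal and thickness textures.
# Emissive combines cubemap and sky light reflections with a two-color gradient driven by fake lighting, tinted by particle color.
# Opacity builds from thickness, with dissolve, edge mask, inner opacity and depth fade branches.
# Dynamic parameters control random UV offset, dissolve, edge offset and panning normal-map UV distortion.
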Material /BRCosmetics/Weapons/FORT_Melee/Pickaxe_PrismParticle/FX/V2/M_Pickaxe_PrismParticle_Impact
graph {
 properties {bUsedWithNiagaraSprites=true, blendMode=BLEND_Translucent, pixelDepthOffsetMode=PDOM_Legacy, shadingModel=MSM_Unlit, usedShadingModels=Unlit}
 nodes [
  n1 = ScalarParameter() {defaultValue=0.1, parameterName=ReflectionHDRIStr}
  n2 = NamedRerouteUsage()
  n3 = ScalarParameter() {parameterName=Reflection_NormalFlatten}
  n4 = ReflectionVectorWS(CustomWorldNormal=Transform(None=MaterialFunctionCall(Normal=@n2, Flatness=@n3) {materialFunction=&FlattenNormal}.Result))
  n5 = TextureSampleParameterCube!(UVs=@n4, Apply View MipBias) {parameterName=ReflectHDRI, samplerType=SAMPLERTYPE_LinearColor, texture=/Game/Effects/FlipSimulation/Rendering/T_WaterReflection}
  n6 = Multiply(A=Multiply(A=Multiply(A=@n1, B=@n5.RGB), B=NamedRerouteUsage()), B=NamedRerouteUsage())
  n7 = VectorParameter() {parameterName=ColorB, defaultValue=<rgba 0 1 0.6667 4>}
  n8 = Add(A=NamedRerouteUsage(), B=ScalarParameter() {group=ThicknessToColor, parameterName=ThicknessToColor_GrayAdjust_Add})
  n9 = ScalarParameter() {defaultValue=1.8, group=ThicknessToColor, parameterName=ThicknessToColor_GrayAdjust_Mul}
  n10 = Power(
   Base=Saturate(None=Multiply(A=@n8, B=@n9))
   Exp=ScalarParameter() {defaultValue=1.2, group=ThicknessToColor, parameterName=ThicknessToColor_GrayAdjust_Power}
  )
  n11 = VectorParameter() {parameterName=FakeLightDirection, defaultValue=<rgba 0 0.5 1 1>}
  n12 = LinearInterpolate(
   A=@n7.RGB
   B=LinearInterpolate(A=VectorParameter() {parameterName=ColorA, defaultValue=<rgba 0 0.6667 0.8 4>}.RGB, B=@n7.RGB, Alpha=Saturate(None=@n10))
   Alpha=DotProduct(A=Normalize(VectorInput=@n11.RGB), B=NamedRerouteUsage())
  )
  n13 = Add(
   A=Multiply(A=Divide(A=@n12), B=NamedRerouteUsage())
   B=Multiply(A=ScalarParameter() {defaultValue=0.1, parameterName=ReflectionSkySTr}, B=SkyLightEnvMapSample(Direction=@n4)) {constA=0.2}
  )
  n14 = ParticleColor()
  MaterialFunctionCall(Emissive=Multiply(A=Add(A=@n6, B=@n13), B=@n14.RGB)) {materialFunction=&MF_FX_EmissiveAdjustments}
  Multiply(A=Constant2Vector() {g=-0.8, r=0.05})
  n15 = LinearInterpolate(B=@n14.A, Alpha=ScalarParameter() {parameterName=UseParticleAlpha}) {constA=1}
  n16 = ScalarParameter() {defaultValue=1.7, parameterName=ThicknessToAlpha_Mul}
  n17 = NamedRerouteUsage()
  n18 = Saturate(None=Add(A=Multiply(A=@n16, B=@n17), B=ScalarParameter() {defaultValue=0.5, parameterName=ThicknessToAlpha_Base}))
  n19 = ScalarParameter() {defaultValue=2.2, parameterName=ThicknessToAlpha_Pow}
  n20 = Add(
   A=NamedRerouteUsage()
   B=Add(
    A=DynamicParameter() {defaultValue=<rgba 0 0 0 1>, paramNames [UV Random, Dissolve, EdgeOffset, UV Distortion]}.EdgeOffset
    B=ScalarParameter() {parameterName=AlphaOffset}
   )
  )
  n21 = ScalarParameter() {defaultValue=2, parameterName=AlphaContrast}
  n22 = Saturate(None=Multiply(
    A=Power(Base=@n18, Exp=@n19)
    B=MaterialFunctionCall(In=@n20, Contrast=@n21) {materialFunction=&CheapContrast}.Result
   ))
  n23 = Multiply(A=Multiply(A=NamedRerouteUsage(), B=NamedRerouteDeclaration(None=@n22) {name=Mask, nodeColor=<rgba 0.0781 0 1 1>}), B=NamedRerouteUsage())
  Multiply(A=Multiply(A=Multiply(A=@n15, B=@n23), B=NamedRerouteUsage()), B=NamedRerouteUsage())
  n24 = DynamicParameter() {defaultValue=<rgba 0 0 0 1>, paramNames [UV Random, Dissolve, EdgeOffset, UV Distortion]}
  n25 = Add(
   A=TextureCoordinate()
   B=MaterialFunctionCall(X=@n24.UV Random, Y=@n24.UV Random) {materialFunction=&MakeFloat2}.Result
  )
  n26 = MaterialFunctionCall(
   X=ScalarParameter() {defaultValue=1, group=UV, parameterName=Distortion Tile X}
   Y=ScalarParameter() {defaultValue=1, group=UV, parameterName=Distortion Tile Y}
  ) {materialFunction=&MakeFloat2}
  n27 = ScalarParameter() {defaultValue=0.05, group=UV, parameterName=Distortion Speed X}
  n28 = ScalarParameter() {defaultValue=0.05, group=UV, parameterName=Distortion Speed Y}
  n29 = TextureSample(UVs=Panner(
    Coordinate=Multiply(A=@n25, B=@n26.Result)
    Time=ParticleRelativeTime()
    Speed=MaterialFunctionCall(X=@n27, Y=@n28) {materialFunction=&MakeFloat2}.Result
   ) {speedX=0.05, speedY=0.5}) {samplerType=SAMPLERTYPE_Normal, texture=&T_Weap_Distort_N_02}
  n30 = Multiply(
   A=MaterialFunctionCall(
    X=ScalarParameter() {defaultValue=0.01, group=UV, parameterName=Distortion Amount X}
    Y=ScalarParameter() {defaultValue=0.01, group=UV, parameterName=Distortion Amount Y}
   ) {materialFunction=&MakeFloat2}.Result
   B=DynamicParameter() {defaultValue=<rgba 0 0 0 1>, paramNames [UV Random, Dissolve, EdgeOffset, UV Distortion]}.UV Distortion
  )
  NamedRerouteDeclaration(None=Multiply(A=ComponentMask(None=@n29.RGB) {g=true, r=true}, B=Multiply(A=@n30))) {name=UV Distort, nodeColor=<rgba 0.5703 1 0 1>}
  n31 = TextureObjectParameter() {
   group=Dissolve
   parameterName=Dissolve_Texture
   samplerType=SAMPLERTYPE_LinearColor
   sortPriority=0
   texture=/Game/Effects/Fort_Effects/Textures/T_FX_IndieBucket_Contrail_Cells
  }
  n32 = DynamicParameter() {defaultValue=<rgba 0 0 0 1>, paramNames [UV Random, Dissolve, EdgeOffset, UV Distortion]}
  n33 = VectorParameter() {group=Dissolve, parameterName=Dissolve_UVControl, channelNames {a=VPan, b=UPan, g=VScale, r=UScale}, defaultValue=<rgba 1 1 0 0>}
  n34 = OneMinus(None=MaterialFunctionCall(
    Texture=@n31
    AdditionalUVOffset=MaterialFunctionCall(X=@n32.UV Random, Y=@n32.UV Random) {materialFunction=&MakeFloat2}.Result
    UVControl_RGB=@n33
    UVControl_A=@n33
   ) {materialFunction=&FXCommon_TextureSample}.R)
  n35 = Add(A=@n32.Dissolve, B=ScalarParameter() {defaultValue=-1, parameterName=DissolveTest})
  NamedRerouteDeclaration(None=MaterialFunctionCall(In=Subtract(A=@n34, B=@n35), Contrast=ScalarParameter() {defaultValue=5, parameterName=DissolveContrast}) {materialFunction=&CheapContrast}.Result) {name=Dissolve, nodeColor=<rgba 0.125 1 0 1>}
  n36 = MaterialFunctionCall(Float2=TextureCoordinate()) {materialFunction=&BreakOutFloat2Components}
  n37 = VectorParameter() {group=EdgeFade, parameterName=EdgeFade_EdgeMask_Control, channelNames {a="", b=Power, g=Multiply, r=Offset}, defaultValue=<rgba 0 30 4 1>}
  NamedRerouteDeclaration(None=Multiply(
    A=MaterialFunctionCall!(UorV=@n36.R, EdgeControl=@n37) {materialFunction=/Game/Effects/Niagara/MasterMaterials/Functions/Common/FXCommon_EdgeMask_Simple}.Mask
    B=MaterialFunctionCall(UorV=@n36.G, EdgeControl=@n37) {materialFunction=&FXCommon_EdgeMask_Simple}.Mask
   )) {name=Edge Mask, nodeColor=<rgba 1 0 0.3281 1>}
  NamedRerouteDeclaration(None=MaterialFunctionCall(
    Value_Morning=ScalarParameter() {defaultValue=0.5, parameterName=ToD_Morning}
    Value_Day=ScalarParameter() {defaultValue=1, parameterName=ToD_Day}
    Value_Evening=ScalarParameter() {defaultValue=0.5, parameterName=ToD_Evening}
    Value_Night=ScalarParameter() {defaultValue=0.5, parameterName=ToD_Night}
   ) {materialFunction=&Scalar_Time_of_Day}.Result) {name=ToD, nodeColor=<rgba 0 0.8594 1 1>}
  n38 = Power(
   Base=DepthFade(FadeDistance=ScalarParameter() {defaultValue=150, parameterName=FadeDistance})
   Exp=ScalarParameter() {defaultValue=2, parameterName=DepthFadePower}
  )
  NamedRerouteDeclaration(None=@n38) {name=DepthFade, nodeColor=<rgba 0 0.9766 1 1>}
  n39 = ScalarParameter() {defaultValue=0.6, parameterName=InnerOpacityA}
  n40 = ScalarParameter() {defaultValue=0.9, parameterName=InnerOpacityB}
  n41 = Add(
   A=Multiply(A=ScalarParameter() {defaultValue=1, parameterName=OpacityToAlpha_Mul}, B=NamedRerouteUsage())
   B=ScalarParameter() {parameterName=OpacityToAlpha_Base}
  )
  n42 = OneMinus(None=Power(Base=Saturate(None=@n41), Exp=ScalarParameter() {defaultValue=3, parameterName=OpacityToAlpha_Pow}))
  NamedRerouteDeclaration(None=LinearInterpolate(B=LinearInterpolate(A=@n39, B=@n40, Alpha=@n42), Alpha=ScalarParameter() {parameterName=UseInnerOpacity}) {constA=1}) {name=InnerOpacity, nodeColor=<rgba 0.9219 1 0 1>}
  n43 = Add(A=TextureCoordinate(), B=NamedRerouteUsage())
  n44 = TextureSampleParameter2D(UVs=@n43) {
   parameterName=NormalTexture
   samplerType=SAMPLERTYPE_Normal
   texture=&/T_FX_Splash_01_N
  }
  NamedRerouteDeclaration(None=@n44.RGB) {name=NormalTexture, nodeColor=<rgba 0 1 0.6484 1>}
  NamedRerouteDeclaration(None=TextureSampleParameter2D(UVs=@n43) {parameterName=ThicknessTexture, texture=&/T_FX_Splash_01}.R) {name=ThicknessTexture, nodeColor=<rgba 1 0.0615 0 1>}
  n45 = Constant() {r=1}
  QualitySwitch(Default=@n38, Low=@n45, High=@n38, Medium=@n45, Epic=@n38)
 ]
}
