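# Niagara swing trail effect for the PrismParticle pickaxe, built from a single DiamondLightTrail emitter with ribbon and sprite renderers.
# Anim-trail modules spawn and update short-lived particles from the User.SkeletalMesh input.
# User.isBusted switches color-over-life between a bright blue-to-pale gradient and a muted, darkening grey-blue one.
# A final multiply halves brightness and color at night and adds an extra alpha fade over a curve.
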NiagaraSystem /BRCosmetics/Weapons/FORT_Melee/Pickaxe_PrismParticle/FX/V2/NS_Pickaxe_PrismParticle_Swing
dependencies {
 usedDataInterfaces [&NiagaraDataInterfaceColorCurve, &NiagaraDataInterfaceCurve]
 usedDynamicInputs []
 usedModules [&SystemState]
}
emitters {DiamondLightTrail {
  emitterUpdateScript [
   &EmitterState()
   &FN_SpawnAnimTrail()
  ]
  particleSpawnScript [
   &FN_InitializeAnimTrailBoneIndecies()
   &FN_InitializeAnimTrail(
    AnimTrailWidth=0.9
    Color=&FN_Select_Color_from_Time_Of_Day(Morning=<rgba 1 1 1 1>, Day=<rgba 1 1 1 1>, Evening=<rgba 1 1 1 1>, Night=<rgba 1 1 1 1>)
    Lifetime=0.1
   )
  ]
  particleUpdateScript [
   &FN_UpdateAnimTrail(SkeletalMesh=$User.SkeletalMesh) {module=FN_UpdateAnimTrail001}
   &Color(Color=&LerpLinearColors(
     StartColor=&ColorFromCurve(ColorCurve=DataInterface() {propertyValues {
        AlphaCurve {keys [<at 0 1>,<at 0.1224 1>,<at 0.3669 0 RCIM_Cubic>]}
        BlueCurve {keys [<at 0.3248 1>,<at 0.716 1>]}
        GreenCurve {keys [<at 0.3248 0.3973>,<at 0.716 0.925>]}
        RedCurve {keys [<at 0.3248 0.0743>,<at 0.716 0.8073>]}
       }})
     EndColor=&ColorFromCurve(ColorCurve=DataInterface() {propertyValues {
        AlphaCurve {keys [<at 0 1>,<at 0.3669 0 RCIM_Cubic>]}
        BlueCurve {keys [<at 0.3248 0.8>,<at 0.716 0.3>]}
        GreenCurve {keys [<at 0.3248 0.6576>,<at 0.716 0.2775>]}
        RedCurve {keys [<at 0.3248 0.64>,<at 0.716 0.2422>]}
       }})
     LerpFactor=&MakeFloatFromInt(Int=$User.isBusted)
    ))
   /SetVariables(Particles.Color=&MultiplyLinearColor(
     Linear Color=&MultiplyLinearColor(
      Linear Color=$Particles.Color
      Scale Factor=&MakeVector4_FromLinearColor(LinearColor=&FNTimeOfDayColor(Night Color=<rgba 0.5 0.5 0.5 1>, Morning Color=<rgba 1 1 1 1>, Day Color=<rgba 1 1 1 1>, Evening Color=<rgba 1 1 1 1>))
     )
     Scale Factor=&MakeVector4(
      X=1
      Y=1
      Z=1
      W=&FloatFromCurve(FloatCurve=DataInterface() {propertyValues {Curve {keys [<at 0 1>,<at 1 0>]}}})
     )
    )) {isSetParameters=true}
   &SolveForcesAndVelocity(Clamp Velocity=false, Limit Acceleration=false)
  ]
  rendererClasses [&NiagaraRibbonRendererProperties, &NiagaraSpriteRendererProperties]
  renderers [
   {rendererClass=&NiagaraRibbonRendererProperties}
   {rendererClass=&NiagaraSpriteRendererProperties, rendererIndex=1}
  ]
 }}
summary {
 systemName=NS_Pickaxe_PrismParticle_Swing
 emitters [{
   emitterName=DiamondLightTrail
   rendererClasses [&NiagaraRibbonRendererProperties, &NiagaraSpriteRendererProperties]
  }]
 userVariables [
  User.DebugVelocity {type=NiagaraFloat, defaultValue=0, description=""}
  User.isBusted {type=NiagaraInt32}
  User.NewVector3f {type=Vector3f, description="", defaultValue {struct=Vector3f, value=<0 0 0>}}
  User.RibbonIDOffset {type=NiagaraInt32}
  User.SkeletalMesh {
   type=NiagaraDataInterfaceSkeletalMesh
   defaultValue {
    struct=NiagaraExt_VariableValue_DataInterface
    value {dataInterface=&self:NiagaraDataInterfaceSkeletalMesh_3}
   }
  }
 ]
}
